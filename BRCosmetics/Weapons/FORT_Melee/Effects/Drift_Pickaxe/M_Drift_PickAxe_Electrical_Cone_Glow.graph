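# Translucent unlit mesh-particle material for an electrical cone glow effect on the Drift pickaxe.
# Color comes from particle color times a boosted VariantColor, shaped by a U-gradient mask on UV1; the V-gradient lerp blends identical inputs, so has no effect.
# Two differently tiled, panning samples of a beam texture are summed, then multiplied by that mask and particle alpha, apparently for opacity.
# Fresnel and both rotator nodes feed nothing shown.
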Material /BRCosmetics/Weapons/FORT_Melee/Effects/Drift_Pickaxe/M_Drift_PickAxe_Electrical_Cone_Glow
graph {
 properties {bUsedWithMeshParticles=true, blendMode=BLEND_Translucent, pixelDepthOffsetMode=PDOM_Legacy, shadingModel=MSM_Unlit, usedShadingModels=Unlit}
 nodes [
  Fresnel()
  n1 = VectorParameter() {parameterName=Electrical_Tiling, defaultValue=<rgba 0.5 2 0 0>}
  n2 = ComponentMask(None=VectorParameter() {parameterName=Electrical_Speed, defaultValue=<rgba -1 0 0 0>}.RGB) {g=true, r=true}
  n3 = Panner(Coordinate=Multiply(A=TextureCoordinate() {coordinateIndex=1}, B=ComponentMask(None=@n1.RGB) {g=true, r=true}), Speed=@n2) {speedY=1}
  MaterialFunctionCall(UVs=@n3, "Rotation Angle (0-1)"=ScalarParameter() {defaultValue=0.25, parameterName="Rotation Angle (0-1) (S)"}) {materialFunction=&CustomRotator}
  n4 = ParticleColor()
  n5 = Multiply(A=VectorParameter() {parameterName=VariantColor, defaultValue=<rgba 1 1 1 1>}.RGB) {constB=250}
  n6 = ComponentMask(None=TextureCoordinate() {coordinateIndex=1}) {g=true}
  n7 = Multiply(A=OneMinus(None=@n6), B=@n6)
  n8 = Multiply(A=@n7, B=@n7) {constB=2}
  n9 = LinearInterpolate(A=@n5, B=@n5, Alpha=Clamp(None=Multiply(A=@n8) {constB=4}))
  n10 = ComponentMask(None=TextureCoordinate() {coordinateIndex=1}) {r=true}
  n11 = Multiply(A=OneMinus(None=@n10), B=Multiply(A=@n10, B=@n10))
  MaterialFunctionCall(Emissive=Multiply(A=Multiply(A=@n4.RGB, B=@n9), B=@n11)) {materialFunction=&MF_FX_EmissiveAdjustments}
  n12 = VectorParameter() {parameterName=Electrical_Tiling_1, defaultValue=<rgba 0.7 2 0 0>}
  n13 = Panner(Coordinate=Multiply(A=TextureCoordinate() {coordinateIndex=1}, B=ComponentMask(None=@n12.RGB) {g=true, r=true}), Speed=@n2) {speedY=1}
  MaterialFunctionCall(UVs=@n13, "Rotation Angle (0-1)"=ScalarParameter() {defaultValue=0.25, parameterName="Rotation Angle (0-1) (S)_1"}) {materialFunction=&CustomRotator}
  n14 = TextureSample(UVs=@n3) {samplerType=SAMPLERTYPE_LinearColor, texture=/Game/Effects/Fort_Effects/Textures/Tile/T_AMR_Beam_02}
  n15 = TextureSample(UVs=@n13) {samplerType=SAMPLERTYPE_LinearColor, texture=&T_AMR_Beam_02}
  Multiply(A=ParticleColor().A, B=Multiply(A=Add(A=@n14.G, B=@n15.R), B=@n11))
 ]
}
